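; Blueprint child of the AutoHeavy pistol weapon that adds a low-ammo audio warning.
; On BeginPlay it runs the parent logic and binds a handler to the local ammo changed event.
; When the magazine holds six or fewer rounds and the weapon is local, the handler plays the semi-auto pistol low-ammo cue as a 2D sound on every such change.
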(event EventBeginPlay
  (|CallParentFunction)
  (EventDispatchers|BindEventtoOnLocalAmmoChanged (delegate Custom|OnLocalAmmoChanged_Event)))

(event EventTick
  (params (DeltaSeconds :type Float%20%28single-precision%29)))

(event Custom|OnLocalAmmoChanged_Event
  (params (LocalCount :type Integer) (LocalRemaining :type Integer))
  (if (<= (FortWeapon|GetLocalMagazineAmmoCount self) 6)
    (if (Variables|Default|GetbIsLocal)
      (Audio|SpawnSound2D "/Game/Sounds/Fort_Weapons/LowAmmo/Pistol_SemiAuto_Low_Ammo_Cue.Pistol_SemiAuto_Low_Ammo_Cue"))))

(fn ConstructionScript
  (|CallParentFunction))
